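; Weapon Blueprint for the Jethro ChillyHammer (Athena test) that adds overheat and cooling feedback.
; Overheating spawns an attached Niagara effect matched to the mesh's first-person primitive type and plays force feedback; finished cooling, equip or unequip destroys the effect.
; Local firing plays an audio layer whose volume follows a curve of the heat ratio.
; Cooling spawns delayed audio if the weapon is visible, and hiding the weapon fades it out.
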(event FortWeapon|EventOnEquippedAndAttached
  (bind _spawned_overheat_fx (Variables|Default|GetSpawned_Overheat_FX))
  ;[0 1 0.4912] On Equip/Unequip -> Reset or else no Overheat FX spawn.
  (|CallParentFunction)
  (Utilities|IsValid _spawned_overheat_fx
    (:Is%20Valid
      (Components|DestroyComponent _spawned_overheat_fx))))

(event FortWeapon|EventOnUnEquippedAndDetached
  (bind _spawned_overheat_fx (Variables|Default|GetSpawned_Overheat_FX))
  ;[0 1 0.4912] On Equip/Unequip -> Reset or else no Overheat FX spawn.
  (|CallParentFunction)
  (Utilities|IsValid _spawned_overheat_fx
    (:Is%20Valid
      (Components|DestroyComponent _spawned_overheat_fx))))

(event EventBeginPlay
  (bind _returnvalue (Game|GetInstigator))
  ;[1 0.6311 0.9916 0.5] Event Begin Play Cast to FortPlayerPawn
  (|CallParentFunction)
  (Utilities|IsValid _returnvalue
    (:Is%20Valid
      (bind _asfort_player_pawn (Utilities|Casting|CastToFortPlayerPawn _returnvalue))
      (Variables|Default|SetAs%20Fort%20Player%20Pawn _asfort_player_pawn))))

(event Overheating|EventOnStartOverheated
  ;[1 0.1083 0] Overheated
  (|CallParentFunction)
  (bind _returnvalue (Niagara|SpawnSystemAttached "/JethroWeaponGameplay/Assets/FX/Weapons/NS_Jethro_ChillyHammer_OverHeat.NS_Jethro_ChillyHammer_OverHeat" (Variables|Visual|GetWeaponMesh) "None" "0, 0, 0" "0, 0, 0" "KeepRelativeOffset" false true "None" false))
  (Variables|Default|SetSpawned_Overheat_FX _returnvalue)
  (Game|Feedback|LocalPlayForceFeedback (Fortnite|GetFortPlayerControllerfromActor (Game|GetInstigator)) "/Game/Weapons/ForceFeedback/FF_Jethro_CosmosSMG_Overheated.FF_Jethro_CosmosSMG_Overheated" false "None")
  (Rendering|SetFirstPersonPrimitiveType (Variables|Default|GetSpawned_Overheat_FX) (Variables|/Script/Engine.PrimitiveComponent|GetFirstPersonPrimitiveType (Variables|Visual|GetWeaponMesh))))

(event Overheating|EventOnFinishedCooling
  ;[0 0.6897 1] Finished Cooling
  (|CallParentFunction)
  (Utilities|FlowControl|Delay 0.2)
  (Components|DestroyComponent (Variables|Default|GetSpawned_Overheat_FX)))

(event Overheating|EventOnStartCooling
  ;[0 1 0.8693] Start Cooling
  (|CallParentFunction)
  (Utilities|Time|ClearandInvalidateTimerbyHandle (Variables|Overheat|GetOverheatFXTimer))
  (Utilities|FlowControl|Delay 1.3)
  (if (Variables|Default|GetVisible)
    (bind _returnvalue (Audio|SpawnSoundAttached "/JethroWeaponGameplay/Assets/Audio/ChillyHammer/Sounds/IP_MSS_Jethro_ChillyHammer_Cooldown.IP_MSS_Jethro_ChillyHammer_Cooldown" (Variables|Visual|GetWeaponMesh)))
    (Variables|Default|SetCoolDownAudio _returnvalue)))

(event EventOnPlayWeaponFireFX
  (params (bPersistentFire :type Boolean) (bSecondaryFire :type Boolean))
  (bind _overheataudiolayer (Variables|Default|GetOverheatAudioLayer))
  ;[1 0.9617 0 0.5] On Weapon Fire
  (|CallParentFunction bPersistentFire bSecondaryFire)
  (bind _as_fort_player_pawn (Variables|Default|GetAs%20Fort%20Player%20Pawn)
    (:then
      (if (Pawn|IsLocallyControlled _as_fort_player_pawn)
        (Audio|Components|Audio|Play _overheataudiolayer 0.0)
        (Audio|Components|Audio|SetVolumeMultiplier _overheataudiolayer (Math|Curves|GetFloatValue (Variables|Overheat|GetOverheatAudioFeedbackCurve) (/ (Overheat|GetCurrentOverheatValue) (Overheat|GetOverheatingMaxValue))) :!class /Script/Engine.AudioComponent)))
    (:else)))

(event EventOnWeaponVisibilityChanged
  (params (bVisible :type Boolean) (bSetForLocalControllerOnly :type Boolean))
  ;[1 1 1] Visibility Handling on Cooldown
  (bind _output_get (Variables|Default|SetVisible bVisible))
  (CallFunction|CoolDownFadeonVisibiltyChange _output_get))

(fn CoolDownFadeOnVisibiltyChange
  (params (Visible :type Boolean))
  (bind _cooldownaudio (Variables|Default|GetCoolDownAudio))
  (if Visible
    _
    (elif (Audio|Components|Audio|IsPlaying _cooldownaudio)
      (Audio|Components|Audio|FadeOut _cooldownaudio 0.3 0.0 "Linear" :!class /Script/Engine.AudioComponent))))

(fn GetActorAudioParameters
  (return 0))

(fn ConstructionScript
  (|CallParentFunction))
